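# M_Asteria_S26Proto_Mat is a simple Default Lit prototype material in an Asteria developer test map folder, flagged for use with Nanite.
# It declares scalar parameters for Specular, Roughness and Metallic, plus vector parameters for BaseColor, defaulting to light cyan, and Emissive Color.
# It contains no textures or math nodes, so material instances can adjust only these flat values.
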Material /Game/Athena/Asteria/Maps/Test/Dev/Stu/M_Asteria_S26Proto_Mat
graph {
 properties {bUsedWithNanite=true, pixelDepthOffsetMode=PDOM_Legacy, usedShadingModels=Default Lit}
 nodes [
  ScalarParameter() {defaultValue=0.5, parameterName=Specular}
  ScalarParameter() {defaultValue=0.5, parameterName=Roughness}
  ScalarParameter() {parameterName=Metallic}
  VectorParameter() {parameterName=BaseColor, defaultValue=<rgba 0.1635 0.9303 1 1>}
  VectorParameter() {parameterName=Emissive Color}
 ]
}
